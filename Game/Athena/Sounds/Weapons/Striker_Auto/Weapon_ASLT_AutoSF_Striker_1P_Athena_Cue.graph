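# First-person fire sound cue for the Striker automatic assault rifle in Athena.
# A Random node picks one of four equally weighted close single-shot waves and feeds it through a Modulator at default settings.
# The Random node is excluded from branch culling.
# The cue uses the first-person ranged-fire sound class, first-person weapon attenuation and the shared first-person gunfire concurrency set.
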SoundCue /Game/Athena/Sounds/Weapons/Striker_Auto/Weapon_ASLT_AutoSF_Striker_1P_Athena_Cue
root=Modulator_0
nodes [
 {id=Modulator_0, children [Random_1]}
 {
  id=Random_1
  children [WavePlayer_0,WavePlayer_5,WavePlayer_6,WavePlayer_7]
  properties {bShouldExcludeFromBranchCulling=true, weights [1,1,1,1]}
 }
 {id=WavePlayer_0, properties {soundWaveAssetPtr=/Game/Athena/Sounds/Weapons/Striker_Auto/Single_Shot/sfx_striker_singlefireNew2_close_01}}
 {id=WavePlayer_5, properties {soundWaveAssetPtr=/Game/Athena/Sounds/Weapons/Striker_Auto/Single_Shot/sfx_striker_singlefireNew2_close_02}}
 {id=WavePlayer_6, properties {soundWaveAssetPtr=/Game/Athena/Sounds/Weapons/Striker_Auto/Single_Shot/sfx_striker_singlefireNew2_close_03}}
 {id=WavePlayer_7, properties {soundWaveAssetPtr=/Game/Athena/Sounds/Weapons/Striker_Auto/Single_Shot/sfx_striker_singlefireNew2_close_04}}
]
properties {
 attenuationSettings=&Fort_Weapon_Projectile_FirstPerson_Athena_Att
 bDebug=true
 maxDistance=2097152
 soundClassObject=&Fort_WeaponRanged_Fire_1P_PSM
 volumeMultiplier=1
 attenuationOverrides {manualReverbSendLevel=-1.0165, reverbSendMethod=Manual}
 concurrencyOverrides {bLimitToOwner=true, maxCount=3, resolutionRule=StopOldest, volumeScale=0.9}
 concurrencySet [&Weapon_Gunfire_FirstPerson_Conc]
}
